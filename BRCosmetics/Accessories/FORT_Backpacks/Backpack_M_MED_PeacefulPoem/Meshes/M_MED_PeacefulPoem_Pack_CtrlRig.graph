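# Control Rig for the PeacefulPoem backpack cosmetic that procedurally waves its tentacle bones.
# The forward solve runs four ChainHarmonics units in sequence on the left and right toxin-jar tentacles, then the left and right mouth beard tentacles.
# Each uses only the wave, with pendulum and reach disabled, and per-chain tuning makes their motion differ.
# The graph drives no other bones, reads no variables, and leaves debug drawing enabled on every unit.
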ControlRigBlueprint /BRCosmetics/Accessories/FORT_Backpacks/Backpack_M_MED_PeacefulPoem/Meshes/M_MED_PeacefulPoem_Pack_CtrlRig
graphs [
 RigVMFunctionLibrary {links [], nodes []}
 RigVMModel {nodes [
   RigUnit_BeginExecution = &RigUnit_BeginExecution()
   /Script/ControlRig.RigUnit_ChainHarmonicsPerItem(
    ChainRoot=(Type=Bone,Name="dyn_tentacle_01_l")
    DrawWorldOffset=(Rotation=(X=0,Y=0,Z=0,W=1),Translation=(X=0,Y=0,Z=0),Scale3D=(X=1,Y=1,Z=1))
    Pendulum=(bEnabled=False,PendulumStiffness=2,PendulumGravity=(X=0,Y=0,Z=0),PendulumBlend=0.75,PendulumDrag=0.98,PendulumMinimum=0,PendulumMaximum=0.1,PendulumEase=Linear,UnwindAxis=(X=0,Y=1,Z=0),UnwindMinimum=0.2,UnwindMaximum=0.05)
    Reach=(bEnabled=False,ReachTarget=(X=0,Y=0,Z=0),ReachAxis=(X=1,Y=0,Z=0),ReachMinimum=0,ReachMaximum=0,ReachEase=Linear)
    Speed=(X=1,Y=1,Z=1)
    Wave=(bEnabled=True,WaveFrequency=(X=1,Y=0.8,Z=-3),WaveAmplitude=(X=0,Y=0,Z=1.5),WaveOffset=(X=0,Y=1,Z=1),WaveNoise=(X=0,Y=0,Z=0),WaveMinimum=0,WaveMaximum=1,WaveEase=Linear)
    WaveCurve=(EditorCurveData=(Keys=((),(Time=0.3914,Value=0.7293))))
    bDrawDebug=True
    ExecutePin=@RigUnit_BeginExecution.ExecutePin
   ) {is=struct RigVMUnitNode}
   > &RigUnit_ChainHarmonicsPerItem(
    ChainRoot=(Type=Bone,Name="dyn_tentacle_01_r")
    DrawWorldOffset=(Rotation=(X=0,Y=0,Z=0,W=1),Translation=(X=0,Y=0,Z=0),Scale3D=(X=1,Y=1,Z=1))
    Pendulum=(bEnabled=False,PendulumStiffness=2,PendulumGravity=(X=0,Y=0,Z=0),PendulumBlend=0.75,PendulumDrag=0.98,PendulumMinimum=0,PendulumMaximum=0.1,PendulumEase=Linear,UnwindAxis=(X=0,Y=1,Z=0),UnwindMinimum=0.2,UnwindMaximum=0.05)
    Reach=(bEnabled=False,ReachTarget=(X=0,Y=0,Z=0),ReachAxis=(X=1,Y=0,Z=0),ReachMinimum=0,ReachMaximum=0,ReachEase=Linear)
    Speed=(X=1.1,Y=1.1,Z=1.1)
    Wave=(bEnabled=True,WaveFrequency=(X=1,Y=0.8,Z=-3),WaveAmplitude=(X=0,Y=0,Z=1.5),WaveOffset=(X=0,Y=1,Z=1),WaveNoise=(X=0,Y=0,Z=0),WaveMinimum=0,WaveMaximum=1,WaveEase=Linear)
    WaveCurve=(EditorCurveData=(Keys=((),(InterpMode=RCIM_Constant,Time=0.6195,Value=0.9894))))
    bDrawDebug=True
   )
   > &RigUnit_ChainHarmonicsPerItem(
    ChainRoot=(Type=Bone,Name="dyn_Tentacle_beard_01_l")
    DrawWorldOffset=(Rotation=(X=0,Y=0,Z=0,W=1),Translation=(X=0,Y=0,Z=0),Scale3D=(X=1,Y=1,Z=1))
    Pendulum=(bEnabled=False,PendulumStiffness=2,PendulumGravity=(X=0,Y=0,Z=0),PendulumBlend=0.75,PendulumDrag=0.98,PendulumMinimum=0,PendulumMaximum=0.1,PendulumEase=Linear,UnwindAxis=(X=0,Y=1,Z=0),UnwindMinimum=0.2,UnwindMaximum=0.05)
    Reach=(bEnabled=False,ReachTarget=(X=0,Y=0,Z=0),ReachAxis=(X=1,Y=0,Z=0),ReachMinimum=0,ReachMaximum=0,ReachEase=Linear)
    Speed=(X=1,Y=1,Z=1.1)
    Wave=(bEnabled=true,WaveFrequency=(X=-1,Y=0.8,Z=-3),WaveAmplitude=(X=0,Y=0,Z=2),WaveOffset=(X=0,Y=1,Z=-0.5),WaveNoise=(X=0,Y=0,Z=0.25),WaveMinimum=0,WaveMaximum=1,WaveEase=Linear)
    WaveCurve=(EditorCurveData=(Keys=((),(Time=0.498,Value=0.452))))
    bDrawDebug=True
   )
   > &RigUnit_ChainHarmonicsPerItem(
    ChainRoot=(Type=Bone,Name="dyn_Tentacle_beard_01_r")
    DrawWorldOffset=(Rotation=(X=0,Y=0,Z=0,W=1),Translation=(X=0,Y=0,Z=0),Scale3D=(X=1,Y=1,Z=1))
    Pendulum=(bEnabled=False,PendulumStiffness=2,PendulumGravity=(X=0,Y=0,Z=0),PendulumBlend=0.75,PendulumDrag=0.98,PendulumMinimum=0,PendulumMaximum=0.1,PendulumEase=Linear,UnwindAxis=(X=0,Y=1,Z=0),UnwindMinimum=0.2,UnwindMaximum=0.05)
    Reach=(bEnabled=False,ReachTarget=(X=0,Y=0,Z=0),ReachAxis=(X=1,Y=0,Z=0),ReachMinimum=0,ReachMaximum=0,ReachEase=Linear)
    Speed=(X=1,Y=1,Z=1.5)
    Wave=(bEnabled=True,WaveFrequency=(X=1,Y=0.8,Z=-3),WaveAmplitude=(X=0,Y=0,Z=1),WaveOffset=(X=0,Y=1,Z=0.5),WaveNoise=(X=0,Y=0,Z=-0.1),WaveMinimum=0,WaveMaximum=1,WaveEase=Linear)
    WaveCurve=(EditorCurveData=(Keys=((),(Time=0.3748,Value=0.5122))))
    bDrawDebug=True
   )
  ]}
]
hierarchy [
 root {loc=<0 0 0>, rot=<0 0 0 1>}
 pack {parent=root, loc=<0 -0.7316 0>, rot=<0 0 0 1>}
 dyn_brain_l {parent=pack, loc=<-12.0075 -6.2656 -1.3809>, rot=<0 0 0 1>}
 dyn_brain_r {parent=pack, loc=<11.4838 -6.2656 -1.3809>, rot=<0 0 0 1>}
 dyn_brain_01_l {parent=pack, loc=<-12.0075 -13.1019 -1.3809>, rot=<0 0 0 1>}
 dyn_brain_01_r {parent=pack, loc=<11.4838 -13.1019 -1.3809>, rot=<0 0 0 1>}
 breathing {parent=pack, loc=<0 -12.6795 -0.7283>, rot=<0 0 0 1>}
 nose {parent=breathing, loc=<-0.0624 -11.1354 -1.6861>, rot=<0 0 0 1>}
 Aya_r {parent=breathing, loc=<3.3155 -7.6439 0.3018>, rot=<0 0 0 1>}
 Aya_l {parent=breathing, loc=<-3.3155 -7.6439 0.3018>, rot=<0 0 0 1>}
 lower_eyeled_r {parent=breathing, loc=<3.3155 -7.5085 0.3018>, rot=<-0.0608 -0.1053 0.4963 -0.8596>}
 upper_eyeled_r {parent=breathing, loc=<3.3155 -7.5085 0.3018>, rot=<0 0 0.5 -0.866>}
 eyebrow_r {parent=upper_eyeled_r, loc=<2.6674 -0.4364 1.1768>, rot=<-0.2042 -0.1623 0.7162 -0.6474>}
 lower_eyeled_l {parent=breathing, loc=<-3.3155 -7.5085 0.3018>, rot=<0 0 0.866 -0.5>}
 upper_eyeled_l {parent=breathing, loc=<-3.3155 -7.5085 0.3018>, rot=<0 0 0.866 -0.5>}
 eyebrow_l {parent=upper_eyeled_l, loc=<2.5367 0.8342 1.0357>, rot=<0.131 -0.2391 -0.6338 -0.7239>}
 cheek_01_l {parent=breathing, loc=<-5.3313 -8.4987 -1.9473>, rot=<0 0 0 1>}
 cheek_01_r {parent=breathing, loc=<5.3595 -8.4987 -1.9473>, rot=<0 0 0 1>}
 jaw_upper {parent=breathing, loc=<0 -6.9204 -2.7288>, rot=<0 0 0 1>}
 cheek_l {parent=jaw_upper, loc=<-7.456 0.0729 -0.1803>, rot=<0 0 0 1>}
 cheek_r {parent=jaw_upper, loc=<8.0604 -0.231 -0.3382>, rot=<0 0 0 1>}
 lips_l {parent=jaw_upper, loc=<-3.6001 -2.5343 -0.5179>, rot=<0 0 0 1>}
 lips_r {parent=jaw_upper, loc=<3.9666 -2.6184 -0.9297>, rot=<0 0 0 1>}
 lip_upper {parent=jaw_upper, loc=<-0.1148 -3.6377 -0.317>, rot=<0 0 0 1>}
 drool_01 {parent=breathing, loc=<-3.5365 -7.5775 -6.0673>, rot=<0 0 0 1>}
 drool_02 {parent=breathing, loc=<1.9626 -8.1111 -5.8644>, rot=<0 0 0 1>}
 drool_04 {parent=breathing, loc=<0.7694 -8.368 -5.738>, rot=<0 0 0 1>}
 drool_05 {parent=breathing, loc=<2.9575 -7.5303 -5.8468>, rot=<0 0 0 1>}
 drool_06 {parent=breathing, loc=<-2.757 -7.6234 -5.7669>, rot=<0 0 0 1>}
 drool_07 {parent=breathing, loc=<-1.7098 -8.4909 -5.0149>, rot=<0 0 0 1>}
 drool_08 {parent=breathing, loc=<-4.8955 -5.6308 -4.8549>, rot=<0 0 0 1>}
 drool_09 {parent=breathing, loc=<5.0747 -6.3568 -5.3659>, rot=<0 0 0 1>}
 lower_skull {parent=root, loc=<0 -0.7316 0>, rot=<0 0 0 1>}
 dyn_toxin_jar {parent=lower_skull, loc=<-0.903 -21.4879 -12.8388>, rot=<0 0 0 1>}
 base_tentacle_r {parent=dyn_toxin_jar, loc=<8.9819 0.1034 -5.3514>, rot=<-0.2315 -0.9725 0.0223 -0.0141>}
 dyn_tentacle_01_r {parent=base_tentacle_r, loc=<0 0 0>, rot=<0 0 0 1>}
 dyn_tentacle_02_r {parent=dyn_tentacle_01_r, loc=<3.5416 0 0>, rot=<-0.6713 0 0 -0.7412>}
 dyn_tentacle_03_r {parent=dyn_tentacle_02_r, loc=<3.4738 0 0>, rot=<0 0 0 1>}
 dyn_tentacle_04_r {parent=dyn_tentacle_03_r, loc=<3.3252 0 0>, rot=<0 0 0 1>}
 base_tentacle_l {parent=dyn_toxin_jar, loc=<-10.8525 -7.2642 -0.5027>, rot=<-0.3223 -0.6668 -0.5888 0.3239>}
 dyn_tentacle_01_l {parent=base_tentacle_l, loc=<0 0 0>, rot=<0 0 0 1>}
 dyn_tentacle_02_l {parent=dyn_tentacle_01_l, loc=<3.1234 0 0>, rot=<-0.6713 0 0 -0.7412>}
 dyn_tentacle_03_l {parent=dyn_tentacle_02_l, loc=<3.1209 0 0>, rot=<0 0 0 1>}
 dyn_tentacle_04_l {parent=dyn_tentacle_03_l, loc=<2.7975 0 0>, rot=<0 0 0 1>}
 mouth_01 {parent=lower_skull, loc=<0 -13.5976 -7.2383>, rot=<-0.0966 -0.0966 0.7005 -0.7005>}
 mouth_02 {parent=mouth_01, loc=<6.1997 0 0>, rot=<0 0 0 1>}
 tongue {parent=mouth_01, loc=<2.2969 -0.0188 2.7295>, rot=<0.0966 0.0966 -0.7005 -0.7005>}
 tongue1 {parent=tongue, loc=<0 -1.0889 -1.1012>, rot=<0 0 0 1>}
 tongue2 {parent=tongue1, loc=<0 -0.9414 -1.1211>, rot=<0 0 0 1>}
 lip_01_r {parent=mouth_01, loc=<5.1148 6.0847 2.6895>, rot=<0.0966 0.0966 -0.7005 -0.7005>}
 lip_02_r {parent=mouth_01, loc=<6.2949 4.329 0.6774>, rot=<0.0966 0.0966 -0.7005 -0.7005>}
 lip_01_l {parent=mouth_01, loc=<5.5369 -5.3117 2.4818>, rot=<0.0966 0.0966 -0.7005 -0.7005>}
 lip_02_l {parent=mouth_01, loc=<6.4477 -3.847 0.4134>, rot=<0.0966 0.0966 -0.7005 -0.7005>}
 lip_lower {parent=mouth_01, loc=<7.1132 0.2994 1.0507>, rot=<0.0966 0.0966 -0.7005 -0.7005>}
 base_Tentacle_beard_r {parent=mouth_01, loc=<7.1613 3.3005 -0.2285>, rot=<0.9833 0.1807 0.0154 -0.0179>}
 dyn_Tentacle_beard_01_r {parent=base_Tentacle_beard_r, loc=<0 0 0>, rot=<0 0 0 1>}
 dyn_Tentacle_beard_02_r {parent=dyn_Tentacle_beard_01_r, loc=<1.5648 0 0>, rot=<0.0072 0.1944 -0.1192 -0.9736>}
 dyn_Tentacle_beard_03_r {parent=dyn_Tentacle_beard_02_r, loc=<1.2323 0 0>, rot=<-0.0449 0.1172 -0.5015 -0.856>}
 dyn_Tentacle_beard_04_r {parent=dyn_Tentacle_beard_03_r, loc=<1.2376 0 0>, rot=<0 0 0 1>}
 base_Tentacle_beard_l {parent=mouth_01, loc=<7.809 -1.8953 -0.3946>, rot=<0.9954 -0.0931 0.0099 -0.0215>}
 dyn_Tentacle_beard_01_l {parent=base_Tentacle_beard_l, loc=<0 0 0>, rot=<0 0 0 1>}
 dyn_Tentacle_beard_02_l {parent=dyn_Tentacle_beard_01_l, loc=<2.2708 0.2425 0.1763>, rot=<0.086 0.1823 -0.1761 -0.9635>}
 dyn_Tentacle_beard_03_l {parent=dyn_Tentacle_beard_02_l, loc=<1.7798 0 -0.0028>, rot=<0 0.104 0 -0.9946>}
]
