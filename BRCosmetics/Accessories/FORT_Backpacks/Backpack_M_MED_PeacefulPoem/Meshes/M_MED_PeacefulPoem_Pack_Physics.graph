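# Physics asset for the PeacefulPoem backpack cosmetic, giving its toxin jar, tentacles and beard tentacles capsule bodies for secondary motion.
# The jar hangs off the pack on a tight swing constraint and carries two four-segment tentacle chains through locked-rotation base bodies.
# A mouth body, rotation-locked to the pack, anchors asymmetric beard chains: three segments left, four right.
# The right beard's root is rotation-locked, unlike the left's.
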PhysicsAsset /BRCosmetics/Accessories/FORT_Backpacks/Backpack_M_MED_PeacefulPoem/Meshes/M_MED_PeacefulPoem_Pack_Physics
bodies {
 base_tentacle_l [{shapeType=Capsule, length=0.505, radius=0.505, shapeName=base_tentacle_l_capsule, rotation=<rot 0 -90 0>}]
 base_tentacle_r [{shapeType=Capsule, length=0.505, radius=0.505, shapeName=base_tentacle_r_capsule, rotation=<rot 0 -90 0>}]
 dyn_Tentacle_beard_01_l [{
   shapeType=Capsule
   length=1.1622
   radius=0.6111
   shapeName=dyn_Tentacle_beard_01_l_capsule
   center=<1.0885 -0.1769 -0.1942>
   rotation=<rot 0.0001 -80.5101 -87.2593>
  }]
 dyn_Tentacle_beard_01_r [{
   shapeType=Capsule
   length=1.8549
   radius=0.5084
   shapeName=dyn_Tentacle_beard_01_r_capsule
   center=<0.6232 -0.0614 -0.2271>
   rotation=<rot -0.0001 -96.6337 -87.3556>
  }]
 dyn_Tentacle_beard_02_l [{
   shapeType=Capsule
   length=1.3208
   radius=0.719
   shapeName=dyn_Tentacle_beard_02_l_capsule
   center=<0.8652 0.032 -0.4973>
   rotation=<rot 0.0001 -63.9236 -85.1401>
  }]
 dyn_Tentacle_beard_02_r [{
   shapeType=Capsule
   length=1.5219
   radius=0.505
   shapeName=dyn_Tentacle_beard_02_r_capsule
   center=<0.5487 -0.2362 -0.1633>
   rotation=<rot -11.1919 92.562 -89.795>
  }]
 dyn_Tentacle_beard_03_l [{
   shapeType=Capsule
   length=0.9398
   radius=0.5823
   shapeName=dyn_Tentacle_beard_03_l_capsule
   center=<1.0511 -0.2353 -0.2502>
   rotation=<rot 0 70.1684 -80.5986>
  }]
 dyn_Tentacle_beard_03_r [{
   shapeType=Capsule
   length=1.4291
   radius=0.5476
   shapeName=dyn_Tentacle_beard_03_r_capsule
   center=<0.3891 -0.1672 -0.2689>
   rotation=<rot 15.1315 -92.2958 -82.5914>
  }]
 dyn_Tentacle_beard_04_r [{
   shapeType=Capsule
   length=1.6376
   radius=0.5355
   shapeName=dyn_Tentacle_beard_04_r_capsule
   center=<0.6232 -0.0066 -0.4133>
   rotation=<rot 0 -83.201 -84.4565>
  }]
 dyn_tentacle_01_l [{shapeType=Capsule, length=1.2029, radius=1.4247, rotation=<rot 88.3689 179.9971 179.9972>}]
 dyn_tentacle_01_r [{
   shapeType=Capsule
   length=1.8633
   radius=1.4881
   shapeName=dyn_tentacle_01_r_capsule
   center=<1.5588 0.1221 -0.1406>
   rotation=<rot -24.4016 91.4637 -93.7378>
  }]
 dyn_tentacle_02_l [{shapeType=Capsule, length=1.9654, radius=1.4729, rotation=<rot 88.3688 179.9961 179.9963>}]
 dyn_tentacle_02_r [{
   shapeType=Capsule
   length=2.8039
   radius=1.4711
   shapeName=dyn_tentacle_02_r_capsule
   center=<1.3344 -0.0741 -0.1966>
   rotation=<rot 5.1882 -86.0296 -87.6926>
  }]
 dyn_tentacle_03_l [{shapeType=Capsule, length=2.235, radius=1.4283, rotation=<rot 88.3688 179.9961 179.9963>}]
 dyn_tentacle_03_r [{
   shapeType=Capsule
   length=2.5832
   radius=1.3314
   shapeName=dyn_tentacle_03_r_capsule
   center=<1.4931 0.0224 -0.1398>
   rotation=<rot 2.553 -87.8267 -88.3681>
  }]
 dyn_tentacle_04_l [{shapeType=Capsule, length=3.3009, radius=1.1157, center=<0.8532 0 0.0243>, rotation=<rot 88.3688 179.9962 179.9964>}]
 dyn_tentacle_04_r [{
   shapeType=Capsule
   length=2.3679
   radius=0.9742
   shapeName=dyn_tentacle_04_r_capsule
   center=<2.4742 0.1695 -0.2936>
   rotation=<rot 0 -88.7671 -89.0576>
  }]
 dyn_toxin_jar [{
   shapeType=Capsule
   length=16
   radius=3.3034
   shapeName=dyn_toxin_jar_capsule
   center=<0.3627 0.3947 -0.6948>
   rotation=<rot -0.3251 98.4105 -83.6359>
  }]
 mouth_02 [{shapeType=Capsule, length=0.505, radius=0.505, shapeName=mouth_02_capsule, rotation=<rot 0 -90 0>}]
 pack [{shapeType=Capsule, length=1, radius=1}]
}
constraints [
 {bone1Name=dyn_tentacle_02_l, bone2Name=dyn_tentacle_01_l, swing1LimitDegrees=35, swing2LimitDegrees=20, twistLimitDegrees=5}
 {bone1Name=dyn_tentacle_03_l, bone2Name=dyn_tentacle_02_l, swing1LimitDegrees=35, swing2LimitDegrees=20, twistLimitDegrees=5}
 {bone1Name=dyn_tentacle_04_l, bone2Name=dyn_tentacle_03_l, swing1LimitDegrees=50.84, swing2LimitDegrees=20, twistLimitDegrees=5}
 {bone1Name=dyn_tentacle_01_l, bone2Name=base_tentacle_l, swing1LimitDegrees=20, swing2LimitDegrees=21.96, twistLimitDegrees=10}
 {
  bone1Name=base_tentacle_l
  bone2Name=dyn_toxin_jar
  swing1LimitDegrees=45
  swing1Motion=Locked
  swing2LimitDegrees=45
  swing2Motion=Locked
  twistLimitDegrees=45
  twistMotion=Locked
 }
 {bone1Name=dyn_tentacle_04_r, bone2Name=dyn_tentacle_03_r, swing1LimitDegrees=30, swing2LimitDegrees=20, twistLimitDegrees=1.36}
 {bone1Name=dyn_tentacle_01_r, bone2Name=base_tentacle_r, swing1LimitDegrees=30, swing2LimitDegrees=20, twistLimitDegrees=10}
 {bone1Name=dyn_tentacle_02_r, bone2Name=dyn_tentacle_01_r, swing1LimitDegrees=30, swing2LimitDegrees=20, twistLimitDegrees=10}
 {bone1Name=dyn_tentacle_03_r, bone2Name=dyn_tentacle_02_r, swing1LimitDegrees=30, swing2LimitDegrees=20, twistLimitDegrees=10}
 {
  bone1Name=base_tentacle_r
  bone2Name=dyn_toxin_jar
  swing1LimitDegrees=45
  swing1Motion=Locked
  swing2LimitDegrees=45
  swing2Motion=Locked
  twistLimitDegrees=45
  twistMotion=Locked
 }
 {bone1Name=dyn_Tentacle_beard_02_l, bone2Name=dyn_Tentacle_beard_01_l, swing1LimitDegrees=26, swing2LimitDegrees=20, twistLimitDegrees=0}
 {bone1Name=dyn_Tentacle_beard_03_l, bone2Name=dyn_Tentacle_beard_02_l, swing1LimitDegrees=26, swing2LimitDegrees=20, twistLimitDegrees=0}
 {bone1Name=dyn_Tentacle_beard_02_r, bone2Name=dyn_Tentacle_beard_01_r, swing1LimitDegrees=26, swing2LimitDegrees=20, twistLimitDegrees=0}
 {bone1Name=dyn_Tentacle_beard_03_r, bone2Name=dyn_Tentacle_beard_02_r, swing1LimitDegrees=26, swing2LimitDegrees=20, twistLimitDegrees=0}
 {bone1Name=dyn_Tentacle_beard_04_r, bone2Name=dyn_Tentacle_beard_03_r, swing1LimitDegrees=26, swing2LimitDegrees=20, twistLimitDegrees=0}
 {bone1Name=dyn_Tentacle_beard_01_l, bone2Name=mouth_02, swing1LimitDegrees=18, swing2LimitDegrees=13, twistLimitDegrees=1}
 {
  bone1Name=dyn_Tentacle_beard_01_r
  bone2Name=mouth_02
  swing1LimitDegrees=45
  swing1Motion=Locked
  swing2LimitDegrees=45
  swing2Motion=Locked
  twistLimitDegrees=45
  twistMotion=Locked
 }
 {bone1Name=dyn_toxin_jar, bone2Name=pack, swing1LimitDegrees=0, swing2LimitDegrees=8, twistLimitDegrees=0}
 {
  bone1Name=mouth_02
  bone2Name=pack
  swing1LimitDegrees=45
  swing1Motion=Locked
  swing2LimitDegrees=45
  swing2Motion=Locked
  twistLimitDegrees=45
  twistMotion=Locked
 }
]
