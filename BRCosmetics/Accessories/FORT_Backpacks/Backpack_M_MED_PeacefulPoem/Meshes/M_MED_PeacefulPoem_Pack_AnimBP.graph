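# Animation Blueprint for the PeacefulPoem backpack cosmetic, adding physics jiggle and idle motion to the pack mesh.
# A non-looping base pose feeds four chained AnimDynamics linear springs on the left and right brain bones, then a RigidBody simulation.
# A Control Rig at partial alpha follows, and a weighted random idle or idle-break sequence is applied additively on top.
# The event graph holds only unconnected stub nodes.
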AnimBlueprint /BRCosmetics/Accessories/FORT_Backpacks/Backpack_M_MED_PeacefulPoem/Meshes/M_MED_PeacefulPoem_Pack_AnimBP
target_skeleton=&/M_MED_PeacefulPoem_Pack_Skeleton
graphs [
 AnimGraph {nodes [
   Input Pose()
   n1 = Sequence Player 'PeacefulPoem_BasePose_PACK'() {bLoopAnimation=false, sequence=/BRCosmetics/Accessories/FORT_Backpacks/Backpack_M_MED_PeacefulPoem/Animation/PeacefulPoem_BasePose_PACK}
   n2 = Anim Dynamics - Bone: dyn_brain_01_l(ComponentPose=Local To Component(LocalPose=@n1), Alpha=1, ExternalForce="0,0,0") {
    angularSpringConstant=1
    bLinearSpring=true
    bOverrideLinearDamping=true
    bUseGravityOverride=true
    linearDampingOverride=1
    linearSpringConstant=380
    simulationSpace=RootRelative
    alphaBoolBlend {blendOption=Linear}
    boundBone {boneName=dyn_brain_01_l}
    componentLinearVelScale=<0.8 0.8 0.8>
    gravityOverride {z=-980}
    physicsBodyDefinitions [{
      boundBone {boneName=dyn_brain_01_l}
      boxExtents=<8 6 8>
      constraintSetup {angularLimitsMax=<0 0 0>, angularLimitsMin=<0 0 0>, angularTarget=<0 0 0>, linearAxesMax=<3 3 5>, linearAxesMin=<-10 -9 -3>}
      localJointOffset=<0 0 0>
     }]
   }
   n3 = Anim Dynamics - Bone: dyn_brain_01_r(ComponentPose=@n2, Alpha=1, ExternalForce="0,0,0") {
    angularSpringConstant=1
    bLinearSpring=true
    bOverrideLinearDamping=true
    bUseGravityOverride=true
    linearDampingOverride=1
    linearSpringConstant=380
    simulationSpace=RootRelative
    alphaBoolBlend {blendOption=Linear}
    boundBone {boneName=dyn_brain_01_r}
    componentLinearVelScale=<0.8 0.8 0.8>
    gravityOverride {z=-980}
    physicsBodyDefinitions [{
      boundBone {boneName=dyn_brain_01_r}
      boxExtents=<10 10 10>
      constraintSetup {angularLimitsMax=<0 0 0>, angularLimitsMin=<0 0 0>, angularTarget=<0 0 0>, linearAxesMax=<8 5 8>, linearAxesMin=<-7 -5 -8>}
      localJointOffset=<0 0 0>
     }]
   }
   n4 = Anim Dynamics - Bone: dyn_brain_l(ComponentPose=@n3, Alpha=1, ExternalForce="0,0,0") {
    angularSpringConstant=1
    bLinearSpring=true
    bOverrideLinearDamping=true
    bUseGravityOverride=true
    linearDampingOverride=1
    linearSpringConstant=380
    simulationSpace=RootRelative
    alphaBoolBlend {blendOption=Linear}
    boundBone {boneName=dyn_brain_l}
    componentLinearVelScale=<0.8 0.8 0.8>
    gravityOverride {z=-980}
    physicsBodyDefinitions [{
      boundBone {boneName=dyn_brain_l}
      boxExtents=<6 6 12>
      constraintSetup {angularLimitsMax=<0 0 0>, angularLimitsMin=<0 0 0>, angularTarget=<0 0 0>, linearAxesMax=<5 6 7>, linearAxesMin=<-10 -6 -7>}
      localJointOffset=<0 0 0>
     }]
   }
   n5 = Anim Dynamics - Bone: dyn_brain_r(ComponentPose=@n4, Alpha=1, ExternalForce="0,0,0") {
    angularSpringConstant=1
    bLinearSpring=true
    bOverrideLinearDamping=true
    bUseGravityOverride=true
    linearDampingOverride=1
    linearSpringConstant=380
    simulationSpace=RootRelative
    alphaBoolBlend {blendOption=Linear}
    boundBone {boneName=dyn_brain_r}
    componentLinearVelScale=<0.8 0.8 0.8>
    gravityOverride {z=-980}
    physicsBodyDefinitions [{
      boundBone {boneName=dyn_brain_r}
      boxExtents=<9 9 9>
      constraintSetup {angularLimitsMax=<0 0 0>, angularLimitsMin=<0 0 0>, angularTarget=<0 0 0>, linearAxesMax=<8 5 7>, linearAxesMin=<-8 -5 -7>}
      localJointOffset=<0 0 0>
     }]
   }
   n6 = RigidBody(ComponentPose=@n5) {
    bOverrideWorldGravity=true
    overridePhysicsAsset=&/M_MED_PeacefulPoem_Pack_Physics
    simulationSpace=BaseBoneSpace
    alphaBoolBlend {blendOption=Linear}
    baseBoneRef {boneName=root}
    componentLinearAccScale=<0.1 0.1 0.1>
    overrideWorldGravity {z=-980}
    simSpaceSettings {maxAngularAcceleration=500, maxAngularVelocity=30, maxLinearAcceleration=450, maxLinearVelocity=250, velocityScaleZ=0, worldAlpha=1}
   }
   n7 = Control Rig(Source=Component To Local(ComponentPose=@n6), Alpha=0.8) {
    alphaBoolBlend {blendOption=Linear}
    controlRigAssetReference {blueprintRigClass=&/M_MED_PeacefulPoem_Pack_CtrlRig.M_MED_PeacefulPoem_Pack_CtrlRig_C}
   }
   n8 = Random Sequence Player!() -> (Pose:Pose Link Structure) {
    is=title AnimGraphNode_RandomPlayer
    entries [
     {
      chanceToPlay=0.8
      maxLoopCount=0
      maxPlayRate=0.8
      minLoopCount=0
      minPlayRate=1
      sequence=/BRCosmetics/Accessories/FORT_Backpacks/Backpack_M_MED_PeacefulPoem/Animation/PeacefulPoem_IdleAnim_PACK
      blendIn {blendOption=Linear, blendTime=0.2, customCurve=null}
     }
     {
      chanceToPlay=0.2
      maxLoopCount=0
      maxPlayRate=1
      minLoopCount=0
      minPlayRate=1
      sequence=/BRCosmetics/Accessories/FORT_Backpacks/Backpack_M_MED_PeacefulPoem/Animation/PeacefulPoem_IdleBreak_PACK
      blendIn {blendOption=Linear, blendTime=0.2, customCurve=null}
     }
    ]
   }
   Output Pose(Result=Apply Additive(Base=@n7, Additive=@n8) {alphaBoolBlend {blendOption=Linear}}) {layerGroup=None}
  ]}
 EventGraph {nodes [
   TryGetPawnOwner()
   Event BlueprintUpdateAnimation()
  ]}
]
variables []
